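; Quest gameplay ability that sends a completion stat event when the player stays near a hostile player long enough after emoting.
; On activation it checks a metre-based radius for hostile pawns, starting a looping timer if any exist and ending otherwise.
; Each tick rechecks, accumulates time in range and reports to the quest manager at the threshold.
; Losing range ends the ability, being downed cancels it, and ending cleans up timer and binding.
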(event Ability|EventActivateAbilityFromEvent
  (params (EventData :type Gameplay%20Event%20Data%20Structure%20%28by%20ref%29))
  ;[0 0 0] Activate ability from event - When the player emotes, check if there is an enemy nearby. If so, set a timer
  (|CallParentFunction EventData)
  (bind _benemiesnearby (CallFunction|CheckforNearbyEnemies))
  (if _benemiesnearby
    (bind _returnvalue (Utilities|Time|SetTimerbyEvent (delegate Custom|OnTimerExecute) (Variables|Default|GetCheckInterval) true))
    (Variables|Default|SetLoopTimer _returnvalue)
    (EventDispatchers|BindEventtoOnPlayerStartDBNO (Variables|GAVars|Data|GetOwningPlayerPawn) (EventDispatchers|CreateEvent :!event K2_CancelAbility))
    (else
      (Development|PrintString "No enemies nearby, ending..." true true "(R=1.000000,G=0.000000,B=0.343959,A=1.000000)" 3.0)
      (Ability|EndAbility))))

(event Custom|OnTimerExecute
  ;[0 0 0] On timer execute - If player is within range of an enemy, and they've been within range for the specified time threshold, send a stat event
  (bind _benemiesnearby (CallFunction|CheckforNearbyEnemies))
  (if _benemiesnearby
    (bind _btimepastthreshold (CallFunction|AppendTime))
    (if _btimepastthreshold
      (Quest|SendComplexCustomStatEvent (Variables|GAVars|Data|GetOwningPlayerQuestManager) 0 (Variables|Default|GetQuestCompletionSourceTag) "(GameplayTags=)" 3)
      (Ability|EndAbility))
    (else
      (Development|PrintString "No enemies nearby, ending..." true true "(R=1.000000,G=0.000000,B=0.343959,A=1.000000)" 3.0)
      (Ability|EndAbility))))

(event Ability|EventOnEndAbility
  (params (bWasCancelled :type Boolean))
  (bind _looptimer (Variables|Default|GetLoopTimer))
  ;[0 0 0] On end ability - unbind events, clear and invalidate timer handle if valid
  (|CallParentFunction bWasCancelled)
  (EventDispatchers|UnbindEventfromOnPlayerStartDBNO (Variables|GAVars|Data|GetOwningPlayerPawn) (EventDispatchers|CreateEvent :!event K2_CancelAbility))
  (if (Utilities|Time|IsValidTimerHandle _looptimer)
    (Utilities|Time|ClearandInvalidateTimerbyHandle _looptimer)
    (Variables|Default|SetTimeInRange 0.0)))

(fn AppendTime
  ;[0 0 0] Append time for how long the player has been within range of an enemy, and check if it meets the time threshold
  (bind _output_get (Variables|Default|SetTimeInRange (+ (Variables|Default|GetTimeInRange) (Variables|Default|GetCheckInterval))))
  (return (>= _output_get (Variables|Default|GetTimeCompletionThreshold))))

(fn CheckForNearbyEnemies
  ;[0 0 0] If there is an enemy player within the specified range, return true
  (LocalVariables|SetbEnemyNearby false)
  (bind _outfortplayerpawns (Fortnite|GetFortPlayerPawnsInRange (Transformation|GetActorLocation (Variables|GAVars|Data|GetOwningPlayerPawn)) (* (Variables|Default|GetRadiusInMeters) 1e+02)))
  (if (Utilities|Array|IsNotEmpty _outfortplayerpawns)
    (bind _array_element (Utilities|Array|ForEachLoopwithBreak _outfortplayerpawns)
      (:LoopBody
        (if (Utilities|Enum|Equal(Enum) (Team|GetTeamAffiliation _array_element (Variables|GAVars|Data|GetOwningPlayerPawn)) "Hostile")
          (LocalVariables|SetbEnemyNearby true)))
      (:Completed
        _))
    (else
      _))
  (return (LocalVariables|GetbEnemyNearby)))
